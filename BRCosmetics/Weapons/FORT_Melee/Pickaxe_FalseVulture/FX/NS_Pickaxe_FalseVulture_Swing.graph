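# Niagara swing trail effect for the FalseVulture pickaxe, drawing short-lived ribbon trails along the weapon's motion.
# Anim-trail emitters spawn ribbons between two bone or socket indices and update them from the User.SkeletalMesh data interface.
# Two emitters are active, a darker trail and a lighter, low-alpha one, each with two ribbon renderers; two similar single-renderer emitters are disabled.
# Color is tinted by time of day, then curves fade it and taper ribbon width over life.
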NiagaraSystem /BRCosmetics/Weapons/FORT_Melee/Pickaxe_FalseVulture/FX/NS_Pickaxe_FalseVulture_Swing
dependencies {
 usedDataInterfaces [&NiagaraDataInterfaceColorCurve, &NiagaraDataInterfaceCurve]
 usedDynamicInputs []
 usedModules [&SystemState]
}
emitters {
 FN_AnimTrail {
  bEnabled=false
  emitterSpawnScript [&FN_RibbonIDOffsetUtility()]
  emitterUpdateScript [
   &EmitterState()
   &FN_SpawnAnimTrail()
  ]
  particleSpawnScript [
   &FN_InitializeAnimTrailBoneIndecies(First Bone or Socket Index=1, Second Bone or Socket Index=0)
   &FN_InitializeAnimTrail(AnimTrailWidth=1.7, Color=<rgba 0.224 0.224 0.224 1>, Lifetime=0.15)
   &Color(Color=&MultiplyLinearColor(
     Linear Color=$Particles.Initial.Color
     Scale Factor=&MakeVector4_FromLinearColor(LinearColor=&FN_Select_Color_from_Time_Of_Day(Morning=<rgba 0.6719 0.6019 0.6097 1>, Day=<rgba 1 1 1 1.1>, Evening=<rgba 1 0.6413 0.7956 1>, Night=<rgba 0.4668 0.9064 2 1>))
    )) {module=Color001}
  ]
  particleUpdateScript [
   &SolveForcesAndVelocity(Clamp Velocity=false, Limit Acceleration=false)
   &FN_UpdateAnimTrail(Locked Particle RibbonWidth:NiagaraFloat=$Particles.RibbonWidth, SkeletalMesh=$User.SkeletalMesh) {module=FN_UpdateAnimTrail001}
   &ScaleColor(
    Linear Color Curve=DataInterface() {propertyValues {
      AlphaCurve {keys [<at 0 0.7>,<at 0.038 1>,<at 0.1439 1>,<at 0.3671 0.6>,<at 0.5309 0.25>,<at 0.7343 0>]}
      BlueCurve {keys [<at 0 0.349>,<at 0.5969 0.5208>]}
      GreenCurve {keys [<at 0 0.349>,<at 0.5969 0.5155>]}
      RedCurve {keys [<at 0 0.349>,<at 0.5969 0.5087>]}
     }}
    Scale Mode=::NewEnumerator2
   )
   &ScaleRibbonWidth(Ribbon Width Scale=&FloatFromCurve(FloatCurve=DataInterface() {propertyValues {Curve {keys [
         {arriveTangent=0.3072, interpMode=RCIM_Cubic, leaveTangent=0.3072, tangentMode=RCTM_User, time=0, value=1}
         {arriveTangent=-0.7535, interpMode=RCIM_Cubic, leaveTangent=-0.7535, tangentMode=RCTM_User, time=1, value=0}
        ]}}}))
  ]
  rendererClasses [&NiagaraRibbonRendererProperties]
  renderers [{rendererClass=&NiagaraRibbonRendererProperties}]
 }
 FN_AnimTrail001 {
  bEnabled=false
  emitterSpawnScript [&FN_RibbonIDOffsetUtility()]
  emitterUpdateScript [
   &EmitterState()
   &FN_SpawnAnimTrail(AnimTrailWidthMode=::FromFirst)
  ]
  particleSpawnScript [
   &FN_InitializeAnimTrailBoneIndecies(First Bone or Socket Index=1, Second Bone or Socket Index=0)
   &FN_InitializeAnimTrail(AnimTrailWidth=0.9, Color=<rgba 0.8396 0.8584 0.8906 0.05>, Lifetime=0.12)
   &Color(Color=&MultiplyLinearColor(
     Linear Color=$Particles.Initial.Color
     Scale Factor=&MakeVector4_FromLinearColor(LinearColor=&FN_Select_Color_from_Time_Of_Day(Morning=<rgba 0.6719 0.6019 0.6097 1>, Day=<rgba 1 1 1 0.9>, Evening=<rgba 1 0.6413 0.7956 1>, Night=<rgba 0.4668 0.9064 2 1>))
    )) {module=Color001}
  ]
  particleUpdateScript [
   &SolveForcesAndVelocity(Clamp Velocity=false, Limit Acceleration=false)
   &FN_UpdateAnimTrail(Locked Particle RibbonWidth:NiagaraFloat=$Particles.RibbonWidth, SkeletalMesh=$User.SkeletalMesh) {module=FN_UpdateAnimTrail001}
   &ScaleColor(
    Linear Color Curve=DataInterface() {propertyValues {
      AlphaCurve {keys [<at 0 0.7>,<at 0.038 1>,<at 0.129 1>,<at 0.2754 0.3>,<at 0.4961 0.05>,<at 0.7442 0>]}
      BlueCurve {keys [<at 0 0.4948>,<at 0.5969 0.5208>]}
      GreenCurve {keys [<at 0 0.4948>,<at 0.5969 0.5208>]}
      RedCurve {keys [<at 0 0.4948>,<at 0.5969 0.5208>]}
     }}
    Scale Mode=::NewEnumerator2
   )
   &ScaleRibbonWidth(Ribbon Width Scale=&FloatFromCurve(FloatCurve=DataInterface() {propertyValues {Curve {keys [
         {arriveTangent=0.3072, interpMode=RCIM_Cubic, leaveTangent=0.3072, tangentMode=RCTM_User, time=0, value=1}
         {arriveTangent=-1.4466, interpMode=RCIM_Cubic, leaveTangent=-1.4466, tangentMode=RCTM_User, time=1, value=0}
        ]}}}))
  ]
  rendererClasses [&NiagaraRibbonRendererProperties]
  renderers [{rendererClass=&NiagaraRibbonRendererProperties}]
 }
 FN_AnimTrail002 {
  emitterSpawnScript [&FN_RibbonIDOffsetUtility()]
  emitterUpdateScript [
   &EmitterState()
   &FN_SpawnAnimTrail()
  ]
  particleSpawnScript [
   &FN_InitializeAnimTrailBoneIndecies(First Bone or Socket Index=1, Second Bone or Socket Index=0)
   &FN_InitializeAnimTrail(Color=<rgba 0.224 0.224 0.224 1>, Lifetime=0.18)
   &Color(Color=&MultiplyLinearColor(
     Linear Color=$Particles.Initial.Color
     Scale Factor=&MakeVector4_FromLinearColor(LinearColor=&FN_Select_Color_from_Time_Of_Day(Morning=<rgba 0.6719 0.6019 0.6097 1>, Day=<rgba 1 1 1 1.1>, Evening=<rgba 1 0.6413 0.7956 1>, Night=<rgba 0.4668 0.9064 2 1>))
    )) {module=Color001}
   &AddVelocity(Velocity=<0 400 0>) {enabled=false}
  ]
  particleUpdateScript [
   &SolveForcesAndVelocity(Clamp Velocity=false, Limit Acceleration=false)
   &FN_UpdateAnimTrail(Locked Particle RibbonWidth=1, SkeletalMesh=$User.SkeletalMesh) {module=FN_UpdateAnimTrail001}
   &ScaleColor(
    Linear Color Curve=DataInterface() {propertyValues {
      AlphaCurve {keys [<at 0 0.7>,<at 0.038 1>,<at 0.1439 1>,<at 0.3671 0.6>,<at 0.5309 0.25>,<at 0.7343 0>]}
      BlueCurve {keys [<at 0 0.349>,<at 0.5969 0.5208>]}
      GreenCurve {keys [<at 0 0.349>,<at 0.5969 0.5155>]}
      RedCurve {keys [<at 0 0.349>,<at 0.5969 0.5087>]}
     }}
    Scale Mode=::NewEnumerator2
   )
   &ScaleRibbonWidth(Ribbon Width Scale=&FloatFromCurve(FloatCurve=DataInterface() {propertyValues {Curve {keys [
         {arriveTangent=0.3072, interpMode=RCIM_Cubic, leaveTangent=0.3072, tangentMode=RCTM_User, time=0, value=1}
         {arriveTangent=-0.7535, interpMode=RCIM_Cubic, leaveTangent=-0.7535, tangentMode=RCTM_User, time=1, value=0}
        ]}}}))
  ]
  rendererClasses [&NiagaraRibbonRendererProperties]
  renderers [
   {rendererClass=&NiagaraRibbonRendererProperties}
   {rendererClass=&NiagaraRibbonRendererProperties, rendererIndex=1}
  ]
 }
 FN_AnimTrail003 {
  emitterSpawnScript [&FN_RibbonIDOffsetUtility()]
  emitterUpdateScript [
   &EmitterState()
   &FN_SpawnAnimTrail()
  ]
  particleSpawnScript [
   &FN_InitializeAnimTrailBoneIndecies(First Bone or Socket Index=1, Second Bone or Socket Index=0)
   &FN_InitializeAnimTrail(Color=<rgba 0.8396 0.8584 0.8906 0.05>, Lifetime=0.16)
   &Color(Color=&MultiplyLinearColor(
     Linear Color=$Particles.Initial.Color
     Scale Factor=&MakeVector4_FromLinearColor(LinearColor=&FN_Select_Color_from_Time_Of_Day(Morning=<rgba 0.6719 0.6019 0.6097 1>, Day=<rgba 1 1 1 0.9>, Evening=<rgba 1 0.6413 0.7956 1>, Night=<rgba 0.4668 0.9064 2 1>))
    )) {module=Color001}
   &AddVelocity(Velocity=<0 400 0>) {enabled=false}
  ]
  particleUpdateScript [
   &SolveForcesAndVelocity(Clamp Velocity=false, Limit Acceleration=false)
   &FN_UpdateAnimTrail(Locked Particle RibbonWidth=1, SkeletalMesh=$User.SkeletalMesh) {module=FN_UpdateAnimTrail001}
   &ScaleColor(
    Linear Color Curve=DataInterface() {propertyValues {
      AlphaCurve {keys [<at 0 0.7>,<at 0.038 1>,<at 0.129 1>,<at 0.2754 0.3>,<at 0.4961 0.05>,<at 0.7442 0>]}
      BlueCurve {keys [<at 0 0.4948>,<at 0.5969 0.5208>]}
      GreenCurve {keys [<at 0 0.4948>,<at 0.5969 0.5208>]}
      RedCurve {keys [<at 0 0.4948>,<at 0.5969 0.5208>]}
     }}
    Scale Mode=::NewEnumerator2
   )
   &ScaleRibbonWidth(Ribbon Width Scale=&FloatFromCurve(FloatCurve=DataInterface() {propertyValues {Curve {keys [
         {arriveTangent=0.3072, interpMode=RCIM_Cubic, leaveTangent=0.3072, tangentMode=RCTM_User, time=0, value=1}
         {arriveTangent=-1.4466, interpMode=RCIM_Cubic, leaveTangent=-1.4466, tangentMode=RCTM_User, time=1, value=0}
        ]}}}))
  ]
  rendererClasses [&NiagaraRibbonRendererProperties]
  renderers [
   {rendererClass=&NiagaraRibbonRendererProperties}
   {rendererClass=&NiagaraRibbonRendererProperties, rendererIndex=1}
  ]
 }
}
summary {
 systemName=NS_Pickaxe_FalseVulture_Swing
 emitters [
  {bEnabled=false, emitterName=FN_AnimTrail, rendererClasses [&NiagaraRibbonRendererProperties]}
  {bEnabled=false, emitterName=FN_AnimTrail001, rendererClasses [&NiagaraRibbonRendererProperties]}
  {emitterName=FN_AnimTrail002, rendererClasses [&NiagaraRibbonRendererProperties]}
  {emitterName=FN_AnimTrail003, rendererClasses [&NiagaraRibbonRendererProperties]}
 ]
 userVariables [
  User.RibbonIDOffset {type=NiagaraInt32}
  User.SkeletalMesh {
   type=NiagaraDataInterfaceSkeletalMesh
   defaultValue {
    struct=NiagaraExt_VariableValue_DataInterface
    value {dataInterface=&self:NiagaraDataInterfaceSkeletalMesh_0}
   }
  }
 ]
}
