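# Physics asset for the FalseVulture axe, a melee pickaxe cosmetic weapon.
# It holds a single body on the b_AR_Root bone, a narrow capsule offset and rotated to lie along the axe.
# There are no constraints, so the axe simulates and collides as one rigid piece.
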PhysicsAsset /BRCosmetics/Weapons/FORT_Melee/Pickaxe_FalseVulture/Meshes/FalseVulture_Axe_PhysicsAsset
bodies {b_AR_Root [{shapeType=Capsule, length=55.8296, radius=4.2929, center=<0.2889 0.0307 40.6789>, rotation=<rot 0.0329 -90 0.0058>}]}
constraints []
